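# Translucent, lit deferred decal material for the PuffyStraw NPC ability FX, forming a round chrome blob with metallic, roughness, specular and base color parameters.
# Time-scrolled radial UVs sample noise, using derivatives from plain UVs for mip selection.
# That noise varies a hard sphere mask's radius, which decal lifetime opacity scales.
# Swirl normals, flattened through a ring mask, blend with a sphere normal map for a domed, rippling surface.
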Material /NPCLibrary/Abilities/PuffyStraw/FX/V2/MD_PuffStraw_Chrome
graph {
 properties {blendMode=BLEND_Translucent, materialDomain=MD_DeferredDecal, pixelDepthOffsetMode=PDOM_Legacy, usedShadingModels=Default Lit}
 nodes [
  n1 = MaterialFunctionCall(Float2=MaterialFunctionCall() {materialFunction=&VectorToRadialValue}.Radial Coordinates) {materialFunction=&BreakOutFloat2Components}
  n2 = Time()
  n3 = Constant() {r=0.1}
  n4 = AppendVector(A=@n1.R, B=Add(A=@n1.G, B=Multiply(A=@n2, B=@n3)))
  n5 = Subtract(A=TextureCoordinate() {uTiling=2, vTiling=2}, B=Constant() {r=1})
  n6 = DDX(Value=@n5)
  n7 = DDY(Value=@n5)
  n8 = TextureSample!(UVs=Multiply(A=@n4, B=Constant2Vector() {g=1.5, r=2}), Tex, "DDX(UVs)"=@n6, "DDY(UVs)"=@n7, Apply View MipBias) {mipValueMode=TMVM_Derivative, samplerType=SAMPLERTYPE_Grayscale, texture=&LowResBlurredNoise}
  Add(A=Divide(A=@n8.R), B=TextureCoordinate())
  AppendVector(A=Constant() {r=1}, B=Constant() {r=1.2})
  Constant()
  n9 = Subtract(A=TextureCoordinate() {uTiling=2, vTiling=2}, B=Constant() {r=1})
  n10 = TextureSample(UVs=@n4, "DDX(UVs)"=DDX(Value=@n9), "DDY(UVs)"=DDY(Value=@n9)) {mipValueMode=TMVM_Derivative, samplerType=SAMPLERTYPE_Grayscale, texture=&LowResBlurredNoise}
  n11 = Add(A=@n10.R, B=Multiply(A=TextureCoordinate()) {constB=2}) {constB=0.5}
  Divide(A=Multiply(
    A=TextureSample!(UVs=@n11, Tex, Apply View MipBias) {texture=&T_NoisePack_01}.RGB
    B=Constant3Vector() {constant=<rgba 1 0.2 1 0>}
   )) {constB=10}
  n12 = Constant() {r=-1}
  n13 = TextureCoordinate()
  n14 = Constant() {r=0.5}
  n15 = Constant()
  n16 = TextureCoordinate()
  n17 = Constant() {r=0.5}
  n18 = Constant() {r=0.3}
  n19 = Constant()
  n20 = Saturate(None=Subtract(A=SphereMask(A=@n13, B=@n14, Radius=@n14, Hardness=@n15), B=SphereMask(A=@n16, B=@n17, Radius=@n18, Hardness=@n19)))
  n21 = MaterialFunctionCall(Texture Scale=Constant() {r=1.1}) {materialFunction=&ScaleUVsByCenter}
  MaterialFunctionCall(
   BaseNormal=MaterialFunctionCall(
    Normal=TextureSample!(UVs=@n4, Tex, "DDX(UVs)"=@n6, "DDY(UVs)"=@n7, Apply View MipBias) {mipValueMode=TMVM_Derivative, samplerType=SAMPLERTYPE_Normal, texture=/Game/Global/Textures/Default/Noise/T_SwirlNoise_N}.RGB
    Flatness=LinearInterpolate(B=@n12, Alpha=@n20) {constA=1}
   ) {materialFunction=&FlattenNormal}.Result
   AdditionalNormal=MaterialFunctionCall(
    Normal=TextureSample!(UVs=@n21.UVs, Tex, Apply View MipBias) {samplerType=SAMPLERTYPE_LinearColor, texture=/Game/Packages/Fortress_SharedMaterials/Textures/SphereNormalsUncomp}.RGB
    Flatness=Constant() {r=-0.2}
   ) {materialFunction=&FlattenNormal}.Result
  ) {materialFunction=&BlendAngleCorrectedNormals}
  MaterialFunctionCall() {materialFunction=&DebugTimeSine}
  ScalarParameter() {defaultValue=0.1, parameterName=Roughness}
  ScalarParameter() {defaultValue=0.95, parameterName=Metallic}
  ScalarParameter() {defaultValue=1, parameterName=Specular}
  n22 = Multiply(A=@n10.R, B=Multiply(B=@n8.R) {constA=2})
  SphereMask(
   A=TextureCoordinate()
   B=Constant() {r=0.5}
   Radius=Multiply(A=LinearInterpolate(Alpha=@n22) {constA=0.3, constB=0.8}, B=DecalLifetimeOpacity().Opacity)
   Hardness=Constant() {r=0.98}
  )
  VectorParameter() {parameterName=Base Color, defaultValue=<rgba 0.6302 0.6302 0.6302 0>}
 ]
}
